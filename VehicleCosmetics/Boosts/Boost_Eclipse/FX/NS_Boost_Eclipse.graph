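# Niagara system for the Eclipse vehicle boost cosmetic, a trail and glow at the boost's position.
# A ribbon emitter draws the trail with a sine-pulsed second layer, while a separate emitter streams fast sparks behind it.
# One GPU emitter merges burst meshes, a glow sprite and drifting sprites through spawn groups and render tags.
# Painted-variant overrides and user parameters drive color and glow, and frontend preview shifts mesh offsets.
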NiagaraSystem /VehicleCosmetics/Boosts/Boost_Eclipse/FX/NS_Boost_Eclipse
dependencies {
 usedDataInterfaces [&NiagaraDataInterfaceCurve, &NiagaraDataInterfaceSkeletalMesh]
 usedDynamicInputs []
 usedModules [
  &self:SystemSpawnScript.SystemScriptSource.SystemScriptGraph.NiagaraNodeAssignment_0.SetVariables_#1
  &SystemState
 ]
}
emitters {
 NE_VC_SingleBoost_Ribbon_Parent002 {
  emitterUpdateScript [
   &EmitterState()
   &VC_FilteredSocketCount()
   &VC_RateQualityHandler(High Spawn Rate=60, Low Spawn Rate=60, Medium Spawn Rate=60, Use Custom Rate?=true)
   &VC_SpawnRibbon_v1(Spawn Time=-1, SpawnRate=$Emitter.Spawn Rate)
  ]
  particleSpawnScript [
   &VC_PaintedVariantsOverride(
    Black=<rgba 0.01 0.01 0.01 1>
    BurntSienna=<rgba 0.022 0.0072 0 1>
    Cobalt=<rgba 0 0.0167 1 1>
    Crimson=<rgba 0.35 0.0128 0.0268 1>
    DefaultColor=<rgba 0.4078 0.1255 0.0314 1>
    ForestGreen=<rgba 0.0154 0.3073 0.0373 1>
    Grey=<rgba 0.0096 0.0104 0.012 1>
    Lime=<rgba 0.0863 0.2235 0 1>
    Orange=<rgba 0.35 0.0308 0.0017 1>
    Pink=<rgba 0.5833 0 0.7 1>
    Purple=<rgba 0.1103 0 0.5625 1>
    R1 Black:LinearColor=<rgba 0.02 0.02 0.02 1>
    R1 BurntSienna:LinearColor=<rgba 0.07 0.0203 0.007 1>
    R1 Cobalt:LinearColor=<rgba 0 0.0185 1 1.5>
    R1 Crimson:LinearColor=<rgba 0.7 0.035 0.0461 1>
    R1 DefaultColor=<rgba 1 0.4 0.12 1>
    R1 Grey:LinearColor=<rgba 0.032 0.032 0.032 1>
    R1 Orange:LinearColor=<rgba 0.5391 0.1265 0 1>
    R1 Pink:LinearColor=<rgba 0.5052 0.15 0.376 1>
    R1 Purple:LinearColor=<rgba 0.1961 0 1 1>
    R1 Saffron:LinearColor=<rgba 0.4 0.4 0.02 1>
    R1 SkyBlue:LinearColor=<rgba 0 0.188 0.875 1>
    R1 TitaniumWhite:LinearColor=<rgba 0 0 0 2>
    Saffron=<rgba 0.3683 0.4 0.02 1>
    SkyBlue=<rgba 0.01 0.1325 0.5 1>
    Use R1 Painted Overrides=true
   )
   &VC_InitializeRibbon_v1(
    Color=$Particles.Color
    Custom Offset=true
    Lifetime=0.25
    Local Vector Offset=&AddVectorToPosition(Position=$System.FinalPosition, Vector=<48 0 0>)
    Offset Amount=1
    Ribbon Width=70
   )
   &AddVelocity(Velocity=<-1500 0 0>, Velocity Speed Scale=3) {module=AddVelocity001}
   &Color(Color=$Particles.Color) {enabled=false}
  ]
  particleUpdateScript [
   &ParticleState()
   &InheritSourceMovement(
    Applied Position Scale=&VectorFromFloat(Value=&FloatFromCurve(FloatCurve=DataInterface() {propertyValues {Curve {keys [<at 0 1>,<at 1 0>]}}}))
    ApplyForce=false
    ApplyPosition=true
    ApplyPositionFalloff=false
   )
   &SolveForcesAndVelocity(Clamp Velocity=false, Limit Acceleration=false)
   &VC_UpdateRibbon_v1()
   &ScaleColor(
    Scale RGB=&VectorFromFloat(Value=&MakeCustomFloatFromBool(Bool=$User.bIsPainted, True Float=$User.Glow, False Float=10))
    ScaleA=true
    ScaleRGB=true
   )
   &DynamicMaterialParameters(
    Index 0 Param 1=&MakeCustomFloatFromBool(Bool=$User.bIsPainted, True Float=0, False Float=1)
    Index 0 Param 2=&MakeCustomFloatFromBool(Bool=$User.bIsPainted, True Float=0, False Float=0)
    Index 0 Param 3=&MakeCustomFloatFromBool(Bool=$User.bIsPainted, True Float=$User.WhiteVersion, False Float=1)
    Index 0 Param 4=&MakeCustomFloatFromBool(Bool=$User.bIsPainted, True Float=$User.AlphaColor, False Float=1)
    Index 1 Param 1=&MakeCustomFloatFromBool(Bool=$User.bIsPainted, True Float=10, False Float=5)
    Module.Index0Float0Write=true
    Module.Index0Float1Write=true
    Module.Index0Float2Write=true
    Module.Index0Float3Write=true
    Module.Index1Float0Write=true
    Module.Index1Float1Write=false
    Module.Index1Float2Write=false
    Module.Index1Float3Write=false
    Write Parameter Index 1=::NewEnumerator1
   ) {enabled=false}
   /SetVariables(
    Particles.Index1:Vector4f=&MakeVector4(X=0, Y=0, Z=0, W=0)
    Particles.RibbonColor1:LinearColor=&VC_DynamicScaleColor(
     Color to Scale:LinearColor=$Particles.Initial.R1 Color
     RGB Scale:Vector3f=&VectorFromFloat(Value=&Multiply_Float(A=5, B=&Sine(Normalized Angle=$Emitter.Age, Period=0.5, Scale=0.3, Bias=0.7)))
     Alpha Scale:NiagaraFloat=&FloatFromCurve(FloatCurve=DataInterface() {propertyValues {Curve {keys [<at 0 1>,<at 0.5 0>]}}}, Scale Curve=2)
    )
    Particles.RibbonPostion1:NiagaraPosition=&AddVectorToPosition(Position=$Particles.Position, Vector=/VehicleCosmetics/FX/Niagara/Modules/VC_ConvertVectorToWorld(Vector:Vector3f=<-25 0 0>))
    Particles.RibbonWidth1=200
   ) {isSetParameters=true}
   &DynamicMaterialParameters(
    Index 0 Param 1=&MakeCustomFloatFromBool(Bool=$User.bIsPainted, True Float=0, False Float=1)
    Index 0 Param 2=&MakeCustomFloatFromBool(Bool=$User.bIsPainted, True Float=0, False Float=0)
    Index 0 Param 3=&MakeCustomFloatFromBool(Bool=$User.bIsPainted, True Float=$User.WhiteVersion, False Float=1)
    Index 0 Param 4=&MakeCustomFloatFromBool(Bool=$User.bIsPainted, True Float=$User.AlphaColor, False Float=1)
    Index 1 Param 1=&MakeCustomFloatFromBool(Bool=$User.bIsPainted, True Float=$User.RibbonEdgeGlow, False Float=1)
    Module.Index0Float0Write=true
    Module.Index0Float1Write=true
    Module.Index0Float2Write=true
    Module.Index0Float3Write=true
    Module.Index1Float0Write=true
    Module.Index1Float1Write=false
    Module.Index1Float2Write=false
    Module.Index1Float3Write=false
    Write Parameter Index 1=::NewEnumerator1
   ) {module=DynamicMaterialParameters001}
  ]
  rendererClasses [&NiagaraRibbonRendererProperties]
  renderers [
   {rendererClass=&NiagaraRibbonRendererProperties}
   {rendererClass=&NiagaraRibbonRendererProperties, rendererIndex=1}
  ]
 }
 NE_VC_SingleBoost_Sprite_Parent {
  simTarget=GPUComputeSim
  emitterSpawnScript [/SetVariables(Emitter.LOCKER:NiagaraBool=&Boolean_NotOperation(Bool=$User.bIsFrontendPreview)) {isSetParameters=true}]
  emitterUpdateScript [
   &EmitterState(
    Inactive Response=::NewEnumerator1
    Life Cycle Mode=::NewEnumerator1
    Loop Behavior=::NewEnumerator1
    Loop Duration Mode=::NewEnumerator1
    UseLoopDelay=false
   )
   &SpawnBurst_Instantaneous(Use Spawn Probability=false, UseLoopCountLimit=false)
   &SpawnBurst_Instantaneous(Spawn Group=1, Use Spawn Probability=false, UseLoopCountLimit=false) {module=SpawnBurst_Instantaneous001}
   &SpawnBurst_Instantaneous(Spawn Group=2, Use Spawn Probability=false, UseLoopCountLimit=false) {module=SpawnBurst_Instantaneous002}
   &SpawnRate(SpawnGroup=3, SpawnRate=20, Use Spawn Probability=false)
   /SetVariables() {isSetParameters=true}
  ]
  particleSpawnScript [
   /SetVariables(Particles.RenderTag:NiagaraInt32=&SpawnGroup()) {isSetParameters=true}
   &VC_PaintedVariantsOverride(
    Black=<rgba 0.007 0.0067 0.0062 1>
    BurntSienna=<rgba 0.025 0.0072 0.0013 1>
    Cobalt=<rgba 0 0.0462 2.5 1>
    Crimson=<rgba 0.9 0 0.018 1>
    DefaultColor=<rgba 1.2 0.525 0.12 1>
    ForestGreen=<rgba 0.0307 0.3073 0.0447 1>
    Grey=<rgba 0.04 0.04 0.04 1>
    Lime=<rgba 0.2917 0.5 0 1>
    Orange=<rgba 2 0.2529 0 1>
    Pink=<rgba 1 0.1458 1 1>
    Purple=<rgba 0.13 0 0.6 1>
    R1 Black:LinearColor=<rgba 0 0 0 1>
    R1 BurntSienna:LinearColor=<rgba 0 0 0 1>
    R1 Cobalt:LinearColor=<rgba 0 0 0 1>
    R1 Crimson:LinearColor=<rgba 0 0 0 1>
    R1 DefaultColor=<rgba 0 0 0 1>
    R1 ForestGreen:LinearColor=<rgba 0 0 0 1>
    R1 Grey:LinearColor=<rgba 0 0 0 1>
    R1 Lime:LinearColor=<rgba 0 0 0 1>
    R1 Orange:LinearColor=<rgba 0 0 0 1>
    R1 Pink:LinearColor=<rgba 0 0 0 1>
    R1 Purple:LinearColor=<rgba 0 0 0 1>
    R1 Saffron:LinearColor=<rgba 0 0 0 1>
    R1 SkyBlue:LinearColor=<rgba 0 0 0 1>
    R3 Black:LinearColor=<rgba 0.3 0.3 0.3 1>
    R3 BurntSienna:LinearColor=<rgba 0.7 0.4083 0 1>
    R3 Cobalt:LinearColor=<rgba 0 0.1 3 1>
    R3 Crimson:LinearColor=<rgba 1.5 0.225 0.2596 1>
    R3 DefaultColor:LinearColor=<rgba 1.5 0.8 0.3 1>
    R3 ForestGreen:LinearColor=<rgba 0.375 1.25 0.4192 1>
    R3 Grey:LinearColor=<rgba 0.5 0.5 0.5 1>
    R3 Lime:LinearColor=<rgba 0.5 1 0 1>
    R3 Orange:LinearColor=<rgba 1.5 0.6749 0.0855 1>
    R3 Pink:LinearColor=<rgba 1.5 0.6 1.5 1>
    R3 Purple:LinearColor=<rgba 0.5437 0.225 1.5 1>
    R3 Saffron:LinearColor=<rgba 0.9 0.9 0 1>
    R3 TitaniumWhite:LinearColor=<rgba 0 0 0 2>
    Saffron=<rgba 0.65 0.6127 0 1>
    SkyBlue=<rgba 0 0.4297 2 1>
    TitaniumWhite=<rgba 0 0 0 1>
    Use R1 Painted Overrides=true
    Use R3 Painted Overrides=true
   )
   &InitializeParticle(
    Color Mode=::NewEnumerator0
    Lifetime=/VehicleCosmetics/FX/Niagara/Modules/MergeEmitters/VC_FloatBySpawnGroup(
     SpawnGroup or ExecIndex:NiagaraInt32=0
     CustomInt:NiagaraInt32=0
     MaximumSpawnGroups:NiagaraInt32=4
     Float SG 0:NiagaraFloat=1
     Float SG 1:NiagaraFloat=1
     Float SG 2:NiagaraFloat=1
     Float SG 3:NiagaraFloat=4
     Float SG 4:NiagaraFloat=0
     Float SG 5:NiagaraFloat=0
    )
    Mesh Scale Mode=::NewEnumerator3
    Mesh Uniform Scale=&VC_FloatBySpawnGroup(
     SpawnGroup or ExecIndex:NiagaraInt32=0
     CustomInt:NiagaraInt32=0
     MaximumSpawnGroups:NiagaraInt32=2
     Float SG 0:NiagaraFloat=0.94
     Float SG 1:NiagaraFloat=0.8
     Float SG 2:NiagaraFloat=0
     Float SG 3:NiagaraFloat=0
     Float SG 4:NiagaraFloat=0
     Float SG 5:NiagaraFloat=0
    )
    Position Offset=&AddVectorToPosition(
     Position=$System.FinalPosition
     Vector=/VehicleCosmetics/FX/Niagara/Modules/MergeEmitters/VC_VectorBySpawnGroup(
      SpawnGroup or ExecIndex:NiagaraInt32=0
      CustomInt:NiagaraInt32=0
      MaximumSpawnGroups:NiagaraInt32=4
      Vector Group 0:Vector3f=&MakeCustomVectorFromBool(Bool=$User.bIsFrontendPreview, True Vector=<-68 0 1>, False Vector=<-74 0 0>)
      Vector Group 1:Vector3f=&MakeCustomVectorFromBool(Bool=$User.bIsFrontendPreview, True Vector=<-65 0 1>, False Vector=<-71 0 0>)
      Vector Group 2:Vector3f=<-65 0 0>
      Vector Group 3:Vector3f=<0 0 0>
      Vector Group 4:Vector3f=<0 0 0>
      Vector Group 5:Vector3f=<0 0 0>
     )
    )
    Position Offset Coordinate Space=::Local
    Sprite Rotation Mode=::NewEnumerator2
    Sprite Size Mode=::NewEnumerator3
    Uniform Sprite Size=&VC_FloatBySpawnGroup(
     SpawnGroup or ExecIndex:NiagaraInt32=0
     CustomInt:NiagaraInt32=0
     MaximumSpawnGroups:NiagaraInt32=4
     Float SG 0:NiagaraFloat=0
     Float SG 1:NiagaraFloat=0
     Float SG 2:NiagaraFloat=80
     Float SG 3:NiagaraFloat=&RandomRangeFloat(Minimum=40, Maximum=5)
     Float SG 4:NiagaraFloat=0
     Float SG 5:NiagaraFloat=0
    )
    UsePositionOffset=true
    Write Material Random=true
   )
   &AddVelocity(Velocity=&VC_VectorBySpawnGroup(
     SpawnGroup or ExecIndex:NiagaraInt32=0
     CustomInt:NiagaraInt32=0
     MaximumSpawnGroups:NiagaraInt32=4
     Vector Group 0:Vector3f=<0 0 0>
     Vector Group 1:Vector3f=<0 0 0>
     Vector Group 2:Vector3f=<0 0 0>
     Vector Group 3:Vector3f=<-300 0 0>
     Vector Group 4:Vector3f=<0 0 0>
     Vector Group 5:Vector3f=<0 0 0>
    ))
   &Color(Color=$Particles.Color)
   &ShapeLocation(
    MaskSpawnGroup=false
    Override Seed=false
    Sphere Radius=&VC_FloatBySpawnGroup(
     SpawnGroup or ExecIndex:NiagaraInt32=0
     CustomInt:NiagaraInt32=0
     MaximumSpawnGroups:NiagaraInt32=4
     Float SG 0:NiagaraFloat=0
     Float SG 1:NiagaraFloat=0
     Float SG 2:NiagaraFloat=0
     Float SG 3:NiagaraFloat=25
     Float SG 4:NiagaraFloat=0
     Float SG 5:NiagaraFloat=0
    )
    UseOwnerScale=true
   )
   /SetVariables(Particles.ColorTag3:LinearColor=$Particles.R3 Color) {isSetParameters=true}
  ]
  particleUpdateScript [
   &ParticleState(Kill Particles When Lifetime Has Elapsed=false)
   &SolveForcesAndVelocity(Clamp Velocity=false, Limit Acceleration=false)
   &ScaleColor(
    Scale Alpha=&VC_FloatBySpawnGroup(
     SpawnGroup or ExecIndex:NiagaraInt32=0
     CustomInt:NiagaraInt32=0
     MaximumSpawnGroups:NiagaraInt32=4
     Float SG 0:NiagaraFloat=1
     Float SG 1:NiagaraFloat=1
     Float SG 2:NiagaraFloat=&MakeCustomFloatFromBool(Bool=$User.bIsPainted, True Float=5, False Float=3)
     Float SG 3:NiagaraFloat=&FloatFromCurve(FloatCurve=DataInterface() {propertyValues {Curve {keys [<at 0 1>,<at 0.5 0>]}}})
     Float SG 4:NiagaraFloat=0
     Float SG 5:NiagaraFloat=0
    )
    Scale RGB=&VectorFromFloat(Value=&VC_FloatBySpawnGroup(
      SpawnGroup or ExecIndex:NiagaraInt32=0
      CustomInt:NiagaraInt32=0
      MaximumSpawnGroups:NiagaraInt32=4
      Float SG 0:NiagaraFloat=&MakeFloatFromVector(
       Vector=&VectorFromFloat(Value=&MakeCustomFloatFromBool(Bool=$User.bIsPainted, True Float=$User.SphereColor, False Float=3))
       Channel=::NewEnumerator0
      )
      Float SG 1:NiagaraFloat=1
      Float SG 2:NiagaraFloat=1
      Float SG 3:NiagaraFloat=3
      Float SG 4:NiagaraFloat=0
      Float SG 5:NiagaraFloat=0
     ))
    ScaleA=true
    ScaleRGB=true
   )
   /SetVariables(Particles.SpriteColor3:LinearColor=&VC_DynamicScaleColor(
     Color to Scale:LinearColor=$Particles.Initial.ColorTag3
     RGB Scale:Vector3f=&VectorFromFloat(Value=&RandomRangeFloat(Minimum=1, Maximum=3))
     Alpha Scale:NiagaraFloat=&FloatFromCurve(
      FloatCurve=DataInterface() {propertyValues {Curve {keys [<at 0.2 0>,<at 0.25 1>,<at 0.5 0>]}}}
      Scale Curve=&RandomRangeFloat(Minimum=0.3, Maximum=1)
     )
    )) {isSetParameters=true}
   &DynamicMaterialParameters(
    Index 0 Param 1=&MakeCustomFloatFromBool(Bool=$User.bIsPainted, True Float=1, False Float=10)
    Module.Index0Float0Write=true
    Module.Index0Float1Write=true
    Module.Index0Float2Write=true
    Module.Index0Float3Write=true
   )
  ]
  rendererClasses [&NiagaraMeshRendererProperties, &NiagaraSpriteRendererProperties]
  renderers [
   {rendererClass=&NiagaraMeshRendererProperties, rendererIndex=0}
   {rendererClass=&NiagaraMeshRendererProperties, rendererIndex=1}
   {rendererClass=&NiagaraSpriteRendererProperties, rendererIndex=2}
   {rendererClass=&NiagaraSpriteRendererProperties, rendererIndex=3}
  ]
 }
 NE_VC_SingleBoost_Sprite_Parent001 {
  emitterSpawnScript [/SetVariables(Emitter.IS_INGAME:NiagaraBool=&Boolean_NotOperation(Bool=$User.bIsFrontendPreview)) {isSetParameters=true}]
  emitterUpdateScript [
   &EmitterState()
   &SpawnRate(SpawnRate=100, Use Spawn Probability=false)
  ]
  particleSpawnScript [
   &VC_PaintedVariantsOverride(
    Black=<rgba 0.3 0.3 0.3 1>
    BurntSienna=<rgba 0.7 0.4083 0 1>
    Cobalt=<rgba 0 0.1 3 1>
    Crimson=<rgba 1.5 0.225 0.2596 1>
    DefaultColor=<rgba 1.5 0.8 0.3 1>
    ForestGreen=<rgba 0.375 1.25 0.4192 1>
    Grey=<rgba 0.5 0.5 0.5 1>
    Lime=<rgba 0.5 1 0 1>
    Orange=<rgba 1.5 0.6749 0.0855 1>
    Pink=<rgba 1.5 0.6 1.5 1>
    Purple=<rgba 0.5437 0.225 1.5 1>
    Saffron=<rgba 0.9 0.9 0 1>
    TitaniumWhite=<rgba 0 0 0 2>
   )
   &InitializeParticle(
    Color Mode=::NewEnumerator0
    Lifetime=0.15
    Position Offset=<-280 0 0>
    Position Offset Coordinate Space=::Local
    Sprite Rotation Mode=::NewEnumerator2
    Sprite Size Mode=::NewEnumerator3
    Uniform Sprite Size=&RandomRangeFloat(Minimum=40, Maximum=5)
    UsePositionOffset=true
    Write Material Random=true
   )
   &AddVelocity(Velocity=<-1500 0 0>, Velocity Speed Scale=3)
   &ShapeLocation(MaskSpawnGroup=false, Override Seed=false, Sphere Radius=30, UseOwnerScale=true)
  ]
  particleUpdateScript [
   &ParticleState()
   &InheritSourceMovement(
    Applied Position Scale=&VectorFromFloat(Value=&FloatFromCurve(FloatCurve=DataInterface() {propertyValues {Curve {keys [<at 0 1.1>,<at 1 0>]}}}))
    ApplyForce=false
    ApplyPosition=true
    ApplyPositionFalloff=false
   )
   &SolveForcesAndVelocity(Clamp Velocity=false, Limit Acceleration=false)
   &ScaleColor(
    Scale Alpha=&FloatFromCurve(FloatCurve=DataInterface() {propertyValues {Curve {keys [{tangentMode=RCTM_User, time=0.5, value=0}, {tangentMode=RCTM_User, time=1, value=1}]}}})
    Scale RGB=&VectorFromFloat(Value=&RandomRangeFloat(Minimum=3, Maximum=5))
    ScaleA=true
    ScaleRGB=true
   )
  ]
  rendererClasses [&NiagaraSpriteRendererProperties]
  renderers [{rendererClass=&NiagaraSpriteRendererProperties}]
 }
}
summary {
 systemName=NS_Boost_Eclipse
 emitters [
  {emitterName=NE_VC_SingleBoost_Ribbon_Parent002, rendererClasses [&NiagaraRibbonRendererProperties]}
  {
   emitterName=NE_VC_SingleBoost_Sprite_Parent
   simTarget=GPUComputeSim
   rendererClasses [&NiagaraMeshRendererProperties, &NiagaraSpriteRendererProperties]
  }
  {emitterName=NE_VC_SingleBoost_Sprite_Parent001, rendererClasses [&NiagaraSpriteRendererProperties]}
 ]
 userVariables [
  User.AlphaColor {type=NiagaraFloat, defaultValue=0, description=""}
  User.bIsFrontendPreview {type=NiagaraBool}
  User.bIsPainted {type=NiagaraBool}
  User.ColorVariantID {type=NiagaraInt32}
  User.DTCarOffset {type=Vector3f, description="", defaultValue {struct=Vector3f, value=<0 0 0>}}
  User.GameModeID {type=NiagaraInt32}
  User.Glow {type=NiagaraFloat, defaultValue=1, description=""}
  User.PaintedColor {type=LinearColor, description="", defaultValue {struct=LinearColor, value=<rgba 1 1 1 1>}}
  User.PortCount {type=NiagaraInt32, defaultValue=1}
  User.RibbonEdgeGlow {type=NiagaraFloat, defaultValue=1, description=""}
  User.SingleBoostPos {type=NiagaraPosition, description="", defaultValue {}}
  User.SphereColor {type=NiagaraFloat, defaultValue=5, description=""}
  User.WhiteVersion {type=NiagaraFloat, defaultValue=1, description=""}
 ]
}
